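# Niagara system for a gold glitter-in-lights environment effect in the HermesGold set, judging by its name and path.
# It holds a single emitter named Bokeh, with SystemState as the only listed module.
# No renderers, data interfaces, dynamic inputs or user variables are listed, so the particle look and behavior are not visible here.
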NiagaraSystem /Game/Effects/Fort_Effects/Effects/Environment/Hermes/HermesGold/FX/NSLW_Env_HermesGold_Glitter_InLights
dependencies {usedDataInterfaces [], usedDynamicInputs [], usedModules [&SystemState]}
emitters {Bokeh {rendererClasses [], renderers []}}
summary {systemName=NSLW_Env_HermesGold_Glitter_InLights, emitters [{emitterName=Bokeh, rendererClasses []}], userVariables []}
