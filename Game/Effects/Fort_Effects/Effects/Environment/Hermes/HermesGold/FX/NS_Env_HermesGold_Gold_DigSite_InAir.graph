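# Niagara environment effect for the HermesGold dig site, filling the air with small, flickering gold dust motes.
# Only Dust_Motes001 is enabled, spawning GPU sprites in a large sphere; debug, glint and other mote emitters are disabled.
# Motes drift under gravity and curl noise, flickering via waveforms, fading by region mask and resizing with camera distance.
# User parameters tune rate, lifetime, size, alpha and flicker, while distance culling limits cost.
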NiagaraSystem /Game/Effects/Fort_Effects/Effects/Environment/Hermes/HermesGold/FX/NS_Env_HermesGold_Gold_DigSite_InAir
dependencies {
 usedDataInterfaces [&NiagaraDataInterfaceCurve, &NiagaraDataInterfaceCamera, &NiagaraDataInterfaceColorCurve]
 usedDynamicInputs [&SimulationPosition]
 usedModules [
  &SystemState
  &self:SystemSpawnScript.SystemScriptSource.SystemScriptGraph.NiagaraNodeAssignment_0.SetVariables_#1
  &self:SystemSpawnScript.SystemScriptSource.SystemScriptGraph.NiagaraNodeAssignment_1.SetVariables_#2
 ]
}
emitters {
 DEBUG {
  bEnabled=false
  emitterUpdateScript [
   &EmitterState()
   &SpawnRate(SpawnRate=500, Use Spawn Probability=false)
  ]
  particleSpawnScript [
   /SetVariables(Particles.RandSpreadCurve:NiagaraFloat=&FloatFromCurve(
     FloatCurve=DataInterface() {propertyValues {Curve {keys [
         {interpMode=RCIM_Cubic, tangentMode=RCTM_User, time=0, value=0.7}
         {interpMode=RCIM_Cubic, tangentMode=RCTM_User, time=0.8, value=0.2}
         {arriveTangent=2.7919, interpMode=RCIM_Cubic, leaveTangent=2.7919, tangentMode=RCTM_User, time=1, value=1}
        ]}}}
     CurveIndex=&RandomRangeFloat(Minimum=0, Maximum=1)
    )) {isSetParameters=true}
   &InitializeParticle(Color=<rgba 0.9565 0 1 1>, Lifetime=5, Sprite Size Mode=::NewEnumerator3, UsePositionOffset=false, Write Material Random=true)
   &ShapeLocation(
    Cone Length:NiagaraFloat=&Lerp_Float(A=200, B=1500, Alpha=$Particles.RandSpreadCurve)
    Cone Mode:ENiagara_ConeMode=::NewEnumerator1
    Cone Surface Distribution=0.5
    MaskSpawnGroup=false
    Override Seed=false
    Rotation Mode=::NewEnumerator1
    Shape Primitive=::NewEnumerator5
    UseOwnerScale=true
    Wedge Horizontal Angle:NiagaraFloat=&Lerp_Float(A=30, B=75, Alpha=$Particles.RandSpreadCurve)
    Wedge Vertical Angle:NiagaraFloat=&Lerp_Float(A=25, B=50, Alpha=$Particles.RandSpreadCurve)
   )
   &ShapeLocation(
    Cone Angle:NiagaraFloat=&Lerp_Float(A=10, B=75, Alpha=$Particles.RandSpreadCurve)
    Cone Inner Angle:NiagaraFloat=&Lerp_Float(A=10, B=10, Alpha=$Particles.RandSpreadCurve)
    Cone Length:NiagaraFloat=&Lerp_Float(A=200, B=1500, Alpha=$Particles.RandSpreadCurve)
    MaskSpawnGroup=false
    Override Seed=false
    Rotation Axis=<0 0 -1>
    Rotation Mode=::NewEnumerator1
    Shape Primitive=::NewEnumerator5
    UseOwnerScale=true
   ) {enabled=false, module=ShapeLocation001}
  ]
  particleUpdateScript [
   /SetVariables(Particles.DEBUG:NiagaraFloat=&VectorLength(Vector=Linked() {linkedVariable {name=Particles.Position, type=NiagaraPosition}})) {isSetParameters=true}
   &ParticleState()
   &ScaleColor(
    Scale Alpha=&Lerp_Float(A=1, B=0, Alpha=$Particles.RandSpreadCurve)
    Scale RGB=&VectorFromFloat(Value=&Lerp_Float(A=1, B=0, Alpha=$Particles.RandSpreadCurve))
    ScaleA=true
    ScaleRGB=true
   )
   &ScaleColor(
    Scale Alpha=&Lerp_Float(
     A=1
     B=0
     Alpha=&RemapRange(
      Input Value To Remap=&VectorLength(Vector=Linked() {linkedVariable {name=Particles.Position, type=NiagaraPosition}})
      Input Min=0
      Input Max=1200
      Output Min=0
      Output Max=1
     )
    )
    Scale RGB=&VectorFromFloat(Value=&Lerp_Float(
      A=1
      B=0
      Alpha=&RemapRange(
       Input Value To Remap=&VectorLength(Vector=Linked() {linkedVariable {name=Particles.Position, type=NiagaraPosition}})
       Input Min=0
       Input Max=1200
       Output Min=0
       Output Max=1
      )
     ))
    ScaleA=true
    ScaleRGB=true
   ) {enabled=false, module=ScaleColor001}
  ]
  rendererClasses [&NiagaraSpriteRendererProperties]
  renderers [{rendererClass=&NiagaraSpriteRendererProperties}]
 }
 Dust_Glints {
  bEnabled=false
  simTarget=GPUComputeSim
  emitterUpdateScript [
   /SetVariables(Emitter.CameraPostion:Vector3f=&ReturnCameraVector(Camera Property=::NewEnumerator0, Coordinate Space=::Simulation)) {isSetParameters=true, module=SetVariables_#3}
   &EmitterState(
    Life Cycle Mode=::NewEnumerator1
    Loop Duration=&RandomRangeFloat(Minimum=5, Maximum=10)
    Recalculate Duration Each Loop=true
    UseLoopDelay=false
   )
   &SpawnRate(SpawnRate=&Lerp_Float(A=0, B=100, Alpha=$NPC.FortniteNPC.IsPlayerIndoors), Use Spawn Probability=false) {enabled=false}
   &SpawnRate(SpawnRate=&Multiply_Float(A=500, B=$System.SpawnrateMult_Glints), Use Spawn Probability=false) {module=SpawnRate001}
   &SpawnRate(SpawnGroup=1, SpawnRate=&Multiply_Float(A=500, B=$System.SpawnrateMult_Glints), Use Spawn Probability=false) {module=SpawnRate002}
   &SpawnRate(SpawnGroup=2, SpawnRate=&Multiply_Float(A=500, B=$System.SpawnrateMult_Glints), Use Spawn Probability=false) {module=SpawnRate003}
   &SpawnRate(SpawnGroup=3, SpawnRate=&Multiply_Float(A=500, B=$System.SpawnrateMult_Glints), Use Spawn Probability=false) {module=SpawnRate004}
   &SpawnRate(SpawnGroup=4, SpawnRate=&Multiply_Float(A=500, B=$System.SpawnrateMult_Glints), Use Spawn Probability=false) {module=SpawnRate005}
  ]
  particleSpawnScript [
   /SetVariables(Particles.RandSpreadCurve:NiagaraFloat=&FloatFromCurve(
     FloatCurve=DataInterface() {propertyValues {Curve {keys [
         {interpMode=RCIM_Cubic, tangentMode=RCTM_User, time=0, value=0.7}
         {interpMode=RCIM_Cubic, tangentMode=RCTM_User, time=0.8, value=0.4}
         {arriveTangent=2.7919, interpMode=RCIM_Cubic, leaveTangent=2.7919, tangentMode=RCTM_User, time=1, value=1}
        ]}}}
     CurveIndex=&RandomRangeFloat(Minimum=0, Maximum=1)
    )) {isSetParameters=true, module=SetVariables_#4}
   &InitializeParticle(
    Color=<rgba 1 0.4669 0.0318 1>
    Lifetime=&UniformRangedFloat(Minimum=4, Maximum=5)
    Mass Mode=::NewEnumerator1
    Sprite Rotation Mode=::NewEnumerator2
    Sprite Size=&Vector2DFromFloat(Value=&RandomRangeFloat(Minimum=1, Maximum=2))
    Sprite Size Mode=::NewEnumerator4
    UsePositionOffset=false
    Write Material Random=true
   ) {module=InitializeParticle001}
   &FN_SetSpawnGroup() {enabled=false}
   &ShapeLocation(MaskSpawnGroup=true, Override Seed=false, Rotation Mode=::NewEnumerator1, Sphere Radius=800, UseOwnerScale=true)
   &ShapeLocation(
    MaskSpawnGroup=true
    Offset=<1500 0 0>
    Offset Mode=::NewEnumerator0
    Override Seed=false
    Rotation Mode=::NewEnumerator1
    Spawn Group Mask=1
    Sphere Radius=800
    UseOwnerScale=true
   ) {module=ShapeLocation001}
   &ShapeLocation(
    MaskSpawnGroup=true
    Offset=<2500 -1500 500>
    Offset Mode=::NewEnumerator0
    Override Seed=false
    Rotation Mode=::NewEnumerator1
    Spawn Group Mask=2
    Sphere Radius=1500
    UseOwnerScale=true
   ) {module=ShapeLocation002}
   &ShapeLocation(
    MaskSpawnGroup=true
    Offset=<2500 -4000 0>
    Offset Mode=::NewEnumerator0
    Override Seed=false
    Rotation Mode=::NewEnumerator1
    Spawn Group Mask=3
    Sphere Radius=800
    UseOwnerScale=true
   ) {module=ShapeLocation004}
   &ShapeLocation(
    MaskSpawnGroup=true
    Offset=<1500 1750 0>
    Offset Mode=::NewEnumerator0
    Override Seed=false
    Rotation Mode=::NewEnumerator1
    Spawn Group Mask=4
    Sphere Radius=800
    UseOwnerScale=true
   ) {module=ShapeLocation005}
   &KillParticlesInVolume(Box Size:Vector3f=<5000 5000 900>, Kill Shape=::NewEnumerator1) {enabled=false}
   &AddVelocity(Velocity=&RandomVector(Vector Scale=&RandomRangeFloat(Minimum=0.5, Maximum=1)))
   /Niagara/Modules/Update/Camera/MaintainInCameraParticleScale()
   &DynamicMaterialParameters(
    Index 0 Param 1=2
    Index 0 Param 2=50
    Index 0 Param 3=200
    Module.Param0WriteEnabled=true
    Module.Param1WriteEnabled=true
    Module.Param2WriteEnabled=true
    Module.Param3WriteEnabled=true
   ) {module=DynamicMaterialParameters001}
   &ShapeLocation(
    MaskSpawnGroup=true
    Override Seed=false
    Spawn Group Mask:NiagaraInt32=&RandomRangeInt(Minimum=0, Maximum=3)
    Sphere Radius=150
    UseOwnerScale=true
   ) {enabled=false, module=ShapeLocation003}
  ]
  particleUpdateScript [
   &ParticleState()
   &self:Camera_to_Particle_Distance_Calculation(Camera Postion:Vector3f=$Emitter.CameraPostion, Distance=3000, Particle Position:NiagaraPosition=$Particles.Position) {module=Camera_to_Particle_Distance_Calculation}
   &ScaleSpriteSize(Scale Factor=&Vector2DFromFloat(Value=&FloatFromCurve(
      FloatCurve=DataInterface() {propertyValues {Curve {keys [<at 0 0.25 RCIM_Cubic>,<at 1 1.5 RCIM_Cubic>]}}}
      CurveIndex=$Particles.NormalizedDistance
     )))
   &Drag(Drag=0.5, Use Linear Drag=true, Use Rotational Drag=false)
   &GravityForce(Gravity=&MakeVector(
     X=0
     Y=0
     Z=&FloatFromCurve(
      FloatCurve=DataInterface() {propertyValues {Curve {keys [<at 0 1 RCIM_Cubic>,<at 1 0>]}}}
      Scale Curve=&RandomRangeFloat(Minimum=-1, Maximum=-3)
     )
    ))
   &CurlNoiseForce(
    Curl Noise Cone Mask Angle=120
    Mask Curl Noise=true
    Noise Frequency=&RandomRangeFloat(Minimum=200, Maximum=400)
    Noise Quality / Cost=::NewEnumerator0
    Noise Strength=&RandomRangeFloat(Minimum=10, Maximum=25)
    Pan Noise=true
    Pan Noise Field=<0 0 1>
   )
   &SpriteRotationRate(Rotation Rate=&RandomRangeFloat(Minimum=-45, Maximum=45))
   &ScaleColor(
    Scale Alpha=&Multiply_Float(
     A=&Lerp_Float(A=1, B=0.3, Alpha=$Particles.RandSpreadCurve)
     B=&Waveform(
      "Waveform [1]"=::NewEnumerator3
      "[1] Frequency"=&RandomRangeFloat(Minimum=0.2, Maximum=0.4)
      "[1] Amplitude Min/Max"=<0 1>
     )
    )
    Scale RGB=&VectorFromFloat(Value=&Lerp_Float(A=1, B=1, Alpha=$Particles.RandSpreadCurve))
    ScaleA=true
    ScaleRGB=true
   ) {module=ScaleColor001}
   &RegionMask(
    Outer Float=0.5
    Region Dimensions:Vector3f=<5000 5000 800>
    Region Type:ENiagaraDebugDrawShape=::NewEnumerator1
    UseOffset=false
    UseRotation=false
   )
   &ScaleColor(
    Scale Alpha=&FloatFromCurve(
     FloatCurve=DataInterface() {propertyValues {Curve {keys [
         {interpMode=RCIM_Cubic, tangentMode=RCTM_User, time=0, value=0}
         {arriveTangent=1.8182, leaveTangent=1.8182, time=0.35, value=1}
         <at 0.55 1>
         <at 1 0>
        ]}}}
     Scale Curve=&Multiply_Float(A=2, B=$Particles.RegionMask.Float)
    )
    Scale RGB=&Multiply_VectorByFloat(Vector=<25 25 25>, Float=1)
    ScaleA=true
    ScaleRGB=true
   )
   &SolveForcesAndVelocity(Clamp Velocity=false, Limit Acceleration=false)
   &DynamicMaterialParameters(
    Index 0 Param 1=0.2
    Index 0 Param 2=&RandomRangeFloat(Minimum=0.2, Maximum=0.4)
    Index 0 Param 3=1
    Module.Index0Param0Write=true
    Module.Index0Param1Write=true
    Module.Index0Param2Write=true
    Module.Index0Param3Write=true
   )
   /SetVariables(
    Particles.Linear Color=&Multiply_LinearColorByFloat(
     Linear Color=$Particles.Color
     Float=&Waveform(
      Global Amplitude Scale=5
      "Waveform [1]"=::NewEnumerator0
      "[1] Frequency"=&RandomRangeFloat(Minimum=0.5, Maximum=1)
      "[1] Amplitude Min/Max"=<0 1>
     )
    )
    Particles.Vector 2D:Vector2f=&Multiply_Vector2D(A=$Particles.SpriteSize, B=<0.25 0.25>)
   ) {isSetParameters=true, module=SetVariables_#5}
  ]
  rendererClasses [&NiagaraSpriteRendererProperties]
  renderers [
   {rendererClass=&NiagaraSpriteRendererProperties}
   {rendererClass=&NiagaraSpriteRendererProperties, rendererIndex=1}
  ]
 }
 Dust_Motes {
  bEnabled=false
  simTarget=GPUComputeSim
  emitterUpdateScript [
   /SetVariables(Emitter.CameraPostion:Vector3f=&ReturnCameraVector(Camera Property=::NewEnumerator0, Coordinate Space=::Simulation)) {isSetParameters=true, module=SetVariables_#3}
   &EmitterState(
    Life Cycle Mode=::NewEnumerator1
    Loop Duration=&RandomRangeFloat(Minimum=5, Maximum=10)
    Recalculate Duration Each Loop=true
    UseLoopDelay=false
   )
   &SpawnRate(SpawnRate=&Lerp_Float(A=0, B=100, Alpha=$NPC.FortniteNPC.IsPlayerIndoors), Use Spawn Probability=false) {enabled=false}
   &SpawnRate(SpawnRate=&Multiply_Float(A=2000, B=$System.SpawnrateMult), Use Spawn Probability=false) {module=SpawnRate001}
   &SpawnRate(SpawnGroup=1, SpawnRate=&Multiply_Float(A=2000, B=$System.SpawnrateMult), Use Spawn Probability=false) {module=SpawnRate002}
   &SpawnRate(SpawnGroup=2, SpawnRate=&Multiply_Float(A=2000, B=$System.SpawnrateMult), Use Spawn Probability=false) {module=SpawnRate003}
   &SpawnRate(SpawnGroup=3, SpawnRate=&Multiply_Float(A=2000, B=$System.SpawnrateMult), Use Spawn Probability=false) {module=SpawnRate004}
   &SpawnRate(SpawnGroup=4, SpawnRate=&Multiply_Float(A=2000, B=$System.SpawnrateMult), Use Spawn Probability=false) {module=SpawnRate005}
  ]
  particleSpawnScript [
   /SetVariables(Particles.RandSpreadCurve:NiagaraFloat=&FloatFromCurve(
     FloatCurve=DataInterface() {propertyValues {Curve {keys [
         {interpMode=RCIM_Cubic, tangentMode=RCTM_User, time=0, value=0.7}
         {interpMode=RCIM_Cubic, tangentMode=RCTM_User, time=0.8, value=0.4}
         {arriveTangent=2.7919, interpMode=RCIM_Cubic, leaveTangent=2.7919, tangentMode=RCTM_User, time=1, value=1}
        ]}}}
     CurveIndex=&RandomRangeFloat(Minimum=0, Maximum=1)
    )) {isSetParameters=true, module=SetVariables_#4}
   &InitializeParticle(
    Color=&MakeLinearColorFromVectorAndFloat(
     "Vector (RGB)"=&MakeVectorFromLinearColorRGB(Color=&MultiplyLinearColor(
       Linear Color=&ColorFromCurve(
        ColorCurve=DataInterface() {propertyValues {
          AlphaCurve {keys []}
          BlueCurve {keys [<at 0.1189 1>,<at 0.1666 0>,<at 0.3109 0.4062>,<at 0.5746 1>,<at 0.8383 0>,<at 0.875 0>,<at 0.9354 1>,<at 1 1>]}
          GreenCurve {keys [<at 0.1189 0.5833>,<at 0.1666 0.1067>,<at 0.3109 0.9>,<at 0.5746 1>,<at 0.8383 0.5>,<at 0.875 0.2>,<at 0.9354 0.2942>,<at 1 0.625>]}
          RedCurve {keys [<at 0.1189 0.5>,<at 0.1666 1>,<at 0.3109 1>,<at 0.5746 1>,<at 0.8383 1>,<at 0.875 1>,<at 0.9354 0.23>,<at 1 0.5>]}
         }}
        CurveIndex=Linked() {linkedVariable {name=NPC.FortniteNPC.TimeOfDay_0-1, type=NiagaraFloat}}
       )
       Scale Factor=&MakeVector4_FromFloat(Float=1)
      ))
     "Float (Alpha)"=&RandomRangeFloat(Minimum=1, Maximum=1)
    )
    Lifetime=&UniformRangedFloat(Minimum=4, Maximum=5)
    Mass Mode=::NewEnumerator1
    Sprite Rotation Mode=::NewEnumerator2
    Sprite Size=&Vector2DFromFloat(Value=&RandomRangeFloat(Minimum=1, Maximum=2))
    Sprite Size Mode=::NewEnumerator4
    UsePositionOffset=false
    Write Material Random=true
   ) {enabled=false, module=InitializeParticle001}
   &InitializeParticle(
    Color=<rgba 1 0.5024 0.0415 1>
    Lifetime=&UniformRangedFloat(Minimum=4, Maximum=5)
    Mass Mode=::NewEnumerator1
    Sprite Rotation Mode=::NewEnumerator2
    Sprite Size=&Vector2DFromFloat(Value=&RandomRangeFloat(Minimum=1, Maximum=2))
    Sprite Size Mode=::NewEnumerator4
    UsePositionOffset=false
    Write Material Random=true
   )
   &FN_SetSpawnGroup() {enabled=false}
   &ShapeLocation(MaskSpawnGroup=true, Override Seed=false, Rotation Mode=::NewEnumerator1, Sphere Radius=800, UseOwnerScale=true)
   &ShapeLocation(
    MaskSpawnGroup=true
    Offset=<1500 0 0>
    Offset Mode=::NewEnumerator0
    Override Seed=false
    Rotation Mode=::NewEnumerator1
    Spawn Group Mask=1
    Sphere Radius=800
    UseOwnerScale=true
   ) {module=ShapeLocation001}
   &ShapeLocation(
    MaskSpawnGroup=true
    Offset=<2500 -1500 500>
    Offset Mode=::NewEnumerator0
    Override Seed=false
    Rotation Mode=::NewEnumerator1
    Spawn Group Mask=2
    Sphere Radius=1500
    UseOwnerScale=true
   ) {module=ShapeLocation002}
   &ShapeLocation(
    MaskSpawnGroup=true
    Offset=<2500 -4000 0>
    Offset Mode=::NewEnumerator0
    Override Seed=false
    Rotation Mode=::NewEnumerator1
    Spawn Group Mask=3
    Sphere Radius=800
    UseOwnerScale=true
   ) {module=ShapeLocation004}
   &ShapeLocation(
    MaskSpawnGroup=true
    Offset=<1500 1750 0>
    Offset Mode=::NewEnumerator0
    Override Seed=false
    Rotation Mode=::NewEnumerator1
    Spawn Group Mask=4
    Sphere Radius=800
    UseOwnerScale=true
   ) {module=ShapeLocation005}
   &KillParticlesInVolume(Box Size:Vector3f=<5000 5000 800>, Kill Shape=::NewEnumerator1) {enabled=false}
   &AddVelocity(Velocity=&RandomVector(Vector Scale=&RandomRangeFloat(Minimum=0.5, Maximum=1)))
   &MaintainInCameraParticleScale()
   &DynamicMaterialParameters(
    Index 0 Param 1=2
    Index 0 Param 2=50
    Index 0 Param 3=200
    Module.Param0WriteEnabled=true
    Module.Param1WriteEnabled=true
    Module.Param2WriteEnabled=true
    Module.Param3WriteEnabled=true
   ) {module=DynamicMaterialParameters001}
   &ShapeLocation(
    MaskSpawnGroup=true
    Override Seed=false
    Spawn Group Mask:NiagaraInt32=&RandomRangeInt(Minimum=0, Maximum=3)
    Sphere Radius=150
    UseOwnerScale=true
   ) {enabled=false, module=ShapeLocation003}
  ]
  particleUpdateScript [
   &ParticleState()
   &self:Camera_to_Particle_Distance_Calculation(Camera Postion:Vector3f=$Emitter.CameraPostion, Distance=1500, Particle Position:NiagaraPosition=$Particles.Position) {module=Camera_to_Particle_Distance_Calculation}
   &ScaleSpriteSize(Scale Factor=&Vector2DFromFloat(Value=&FloatFromCurve(
      FloatCurve=DataInterface() {propertyValues {Curve {keys [<at 0 0.25 RCIM_Cubic>,<at 1 1.5 RCIM_Cubic>]}}}
      CurveIndex=$Particles.NormalizedDistance
     )))
   &Drag(Drag=0.5, Use Linear Drag=true, Use Rotational Drag=false)
   &GravityForce(Gravity=&MakeVector(
     X=0
     Y=0
     Z=&FloatFromCurve(
      FloatCurve=DataInterface() {propertyValues {Curve {keys [<at 0 1 RCIM_Cubic>,<at 1 0>]}}}
      Scale Curve=&RandomRangeFloat(Minimum=-1, Maximum=-3)
     )
    ))
   &CurlNoiseForce(
    Curl Noise Cone Mask Angle=120
    Mask Curl Noise=true
    Noise Frequency=&RandomRangeFloat(Minimum=200, Maximum=400)
    Noise Quality / Cost=::NewEnumerator0
    Noise Strength=&RandomRangeFloat(Minimum=10, Maximum=25)
    Pan Noise=true
    Pan Noise Field=<0 0 1>
   )
   &SpriteRotationRate(Rotation Rate=&RandomRangeFloat(Minimum=-45, Maximum=45))
   &ScaleColor(
    Scale Alpha=&Multiply_Float(
     A=&Lerp_Float(A=1, B=0.3, Alpha=$Particles.RandSpreadCurve)
     B=&Waveform(
      "Waveform [1]"=::NewEnumerator3
      "[1] Frequency"=&RandomRangeFloat(Minimum=0.2, Maximum=0.4)
      "[1] Amplitude Min/Max"=<0 1>
     )
    )
    Scale RGB=&VectorFromFloat(Value=&Lerp_Float(A=1, B=1, Alpha=$Particles.RandSpreadCurve))
    ScaleA=true
    ScaleRGB=true
   ) {module=ScaleColor001}
   &RegionMask(
    Outer Float=0.5
    Region Dimensions:Vector3f=<5000 5000 800>
    Region Type:ENiagaraDebugDrawShape=::NewEnumerator1
    UseOffset=false
    UseRotation=false
   )
   &ScaleColor(
    Scale Alpha=&FloatFromCurve(
     FloatCurve=DataInterface() {propertyValues {Curve {keys [
         {interpMode=RCIM_Cubic, tangentMode=RCTM_User, time=0, value=0}
         {arriveTangent=1.8182, leaveTangent=1.8182, time=0.35, value=1}
         <at 0.55 1>
         <at 1 0>
        ]}}}
     Scale Curve=&Multiply_Float(A=2, B=$Particles.RegionMask.Float)
    )
    Scale RGB=&Multiply_VectorByFloat(Vector=<25 25 25>, Float=1)
    ScaleA=true
    ScaleRGB=true
   )
   &SolveForcesAndVelocity(Clamp Velocity=false, Limit Acceleration=false)
   &DynamicMaterialParameters(
    Index 0 Param 1=1
    Index 0 Param 2=150
    Index 0 Param 3=150
    Index 1 Param 1=1
    Module.Index0Param0Write=true
    Module.Index0Param1Write=true
    Module.Index0Param2Write=true
    Module.Index0Param3Write=true
    Module.Index1Param0Write=true
    Module.Index1Param1Write=true
    Module.Index1Param2Write=false
    Module.Index1Param3Write=true
    Write Parameter Index 1=true
   )
   /SetVariables(
    Particles.Linear Color=&Multiply_LinearColorByFloat(
     Linear Color=$Particles.Color
     Float=&Waveform(
      Global Amplitude Scale=5
      "Waveform [1]"=::NewEnumerator0
      "[1] Frequency"=&RandomRangeFloat(Minimum=0.5, Maximum=2)
      "[1] Amplitude Min/Max"=<0 1>
     )
    )
    Particles.Vector 2D:Vector2f=&Multiply_Vector2D(A=$Particles.SpriteSize, B=<0.25 0.25>)
   ) {isSetParameters=true, module=SetVariables_#5}
  ]
  rendererClasses [&NiagaraSpriteRendererProperties]
  renderers [{rendererClass=&NiagaraSpriteRendererProperties}]
 }
 Dust_Motes001 {
  simTarget=GPUComputeSim
  emitterUpdateScript [
   /SetVariables(Emitter.CameraPostion:Vector3f=&ReturnCameraVector(Camera Property=::NewEnumerator0, Coordinate Space=::Simulation)) {isSetParameters=true, module=SetVariables_#3}
   &EmitterState(
    Enable Distance Culling=true
    Life Cycle Mode=::NewEnumerator1
    Loop Duration=&RandomRangeFloat(Minimum=1, Maximum=2)
    Max Distance Response=::SleepAndClearParticles
    MaxDistance=3000
    Module.bApplySpawnCountScaleByDistance=true
    Module.bUseMaxDistance=true
    Module.bUseMinDistance=false
    Recalculate Duration Each Loop=true
    Scalability Mode=::NewEnumerator1
    Scale Spawn Count=true
    Spawn Count Scale By Distance=DataInterface() {propertyValues {Curve {keys [<at 0 1>,<at 0.8 1>,<at 0.9962 0.4982>]}}}
    UseLoopDelay=false
   )
   &SpawnRate(SpawnRate=&Multiply_Float(A=12500, B=$User.SpawnRateMult), Use Spawn Probability=false)
  ]
  particleSpawnScript [
   /SetVariables(Particles.RandSpreadCurve:NiagaraFloat=&FloatFromCurve(
     FloatCurve=DataInterface() {propertyValues {Curve {keys [
         {interpMode=RCIM_Cubic, tangentMode=RCTM_User, time=0, value=0.7}
         {interpMode=RCIM_Cubic, tangentMode=RCTM_User, time=0.8, value=0.4}
         {arriveTangent=2.7919, interpMode=RCIM_Cubic, leaveTangent=2.7919, tangentMode=RCTM_User, time=1, value=1}
        ]}}}
     CurveIndex=&RandomRangeFloat(Minimum=0, Maximum=1)
    )) {isSetParameters=true, module=SetVariables_#4}
   &InitializeParticle(
    Color=&MakeLinearColorFromVectorAndFloat(
     "Vector (RGB)"=&MakeVectorFromLinearColorRGB(Color=&MultiplyLinearColor(
       Linear Color=&ColorFromCurve(
        ColorCurve=DataInterface() {propertyValues {
          AlphaCurve {keys []}
          BlueCurve {keys [<at 0.1189 1>,<at 0.1666 0>,<at 0.3109 0.4062>,<at 0.5746 1>,<at 0.8383 0>,<at 0.875 0>,<at 0.9354 1>,<at 1 1>]}
          GreenCurve {keys [<at 0.1189 0.5833>,<at 0.1666 0.1067>,<at 0.3109 0.9>,<at 0.5746 1>,<at 0.8383 0.5>,<at 0.875 0.2>,<at 0.9354 0.2942>,<at 1 0.625>]}
          RedCurve {keys [<at 0.1189 0.5>,<at 0.1666 1>,<at 0.3109 1>,<at 0.5746 1>,<at 0.8383 1>,<at 0.875 1>,<at 0.9354 0.23>,<at 1 0.5>]}
         }}
        CurveIndex=Linked() {linkedVariable {name=NPC.FortniteNPC.TimeOfDay_0-1, type=NiagaraFloat}}
       )
       Scale Factor=&MakeVector4_FromFloat(Float=1)
      ))
     "Float (Alpha)"=&RandomRangeFloat(Minimum=1, Maximum=1)
    )
    Lifetime=&UniformRangedFloat(Minimum=4, Maximum=5)
    Mass Mode=::NewEnumerator1
    Sprite Rotation Mode=::NewEnumerator2
    Sprite Size=&Vector2DFromFloat(Value=&RandomRangeFloat(Minimum=1, Maximum=2))
    Sprite Size Mode=::NewEnumerator4
    UsePositionOffset=false
    Write Material Random=true
   ) {enabled=false, module=InitializeParticle001}
   &InitializeParticle(
    Color Maximum=$System.ColorNear
    Color Minimum=$System.ColorFar
    Color Mode=::NewEnumerator2
    Lifetime=&Multiply_Float(A=&UniformRangedFloat(Minimum=1.65, Maximum=3), B=$User.LifeTimeMult)
    Mass Mode=::NewEnumerator1
    Sprite Rotation Mode=::NewEnumerator2
    Sprite Size=&Vector2DFromFloat(Value=&Multiply_Float(A=&RandomRangeFloat(Minimum=2, Maximum=5), B=$User.Size))
    Sprite Size Mode=::NewEnumerator4
    UsePositionOffset=false
    Write Material Random=true
   )
   &FN_SetSpawnGroup() {enabled=false}
   &ShapeLocation(MaskSpawnGroup=true, Override Seed=false, Rotation Mode=::NewEnumerator1, Sphere Radius=2500, UseOwnerScale=true)
   &KillParticlesInVolume(Box Size:Vector3f=<5000 5000 800>, Kill Shape=::NewEnumerator1) {enabled=false}
   &AddVelocity(Velocity=&RandomVector(Vector Scale=&RandomRangeFloat(Minimum=0.5, Maximum=1)))
   &MaintainInCameraParticleScale()
   &DynamicMaterialParameters(
    Index 0 Param 1=2
    Index 0 Param 2=50
    Index 0 Param 3=200
    Module.Index0Float0Write=true
    Module.Index0Float1Write=true
    Module.Index0Float2Write=true
    Module.Index0Float3Write=true
   ) {module=DynamicMaterialParameters001}
   &ShapeLocation(
    MaskSpawnGroup=true
    Override Seed=false
    Spawn Group Mask:NiagaraInt32=&RandomRangeInt(Minimum=0, Maximum=3)
    Sphere Radius=150
    UseOwnerScale=true
   ) {enabled=false, module=ShapeLocation003}
  ]
  particleUpdateScript [
   &ParticleState()
   &ScaleSpriteSize(
    Uniform Curve Scale=&FN_LOD_Distance_Remap(Start Distance=200, End Distance=3000, Minimum Output=1, Maximum Output=0.25, Clamp Output=true)
    Uniform Curve Sprite Scale=DataInterface() {propertyValues {Curve {keys [
        {interpMode=RCIM_Cubic, tangentMode=RCTM_User, time=0, value=0}
        <at 0.2 1>
        {arriveTangent=-1.3125, leaveTangent=-1.3125, time=0.7, value=1}
        {tangentMode=RCTM_User, time=1, value=0}
       ]}}}
   ) {module=ScaleSpriteSize001}
   &self:Camera_to_Particle_Distance_Calculation(Camera Postion:Vector3f=$Emitter.CameraPostion, Distance=1500, Particle Position:NiagaraPosition=$Particles.Position) {module=Camera_to_Particle_Distance_Calculation}
   &ScaleSpriteSize(Scale Factor=&Vector2DFromFloat(Value=&FloatFromCurve(
      FloatCurve=DataInterface() {propertyValues {Curve {keys [<at 0 0.25 RCIM_Cubic>,<at 1 1.5 RCIM_Cubic>]}}}
      CurveIndex=$Particles.NormalizedDistance
     )))
   &GravityForce(Gravity=&MakeVector(
     X=0
     Y=0
     Z=&FloatFromCurve(
      FloatCurve=DataInterface() {propertyValues {Curve {keys [<at 0 1 RCIM_Cubic>,<at 1 0>]}}}
      Scale Curve=&RandomRangeFloat(Minimum=-2, Maximum=-10)
     )
    ))
   &Drag(Drag=0.5, Use Linear Drag=true, Use Rotational Drag=false)
   &CurlNoiseForce(
    Noise Frequency=&RandomRangeFloat(Minimum=0.2, Maximum=1)
    Noise Strength=&RandomRangeFloat(Minimum=10, Maximum=25)
    Pan Noise=true
    Pan Noise Field=<0 0 1>
   )
   &SpriteRotationRate(Rotation Rate=&RandomRangeFloat(Minimum=-45, Maximum=45))
   &ScaleColor(
    Scale RGB=&MakeVectorFromLinearColorRGB(Color=&HueShiftLinearColor(
      Hue=0
      Saturation=&Waveform(
       Global Amplitude Scale=2
       "Waveform [1]"=::NewEnumerator0
       "[1] Frequency"=&RandomRangeFloat(Minimum=0.5, Maximum=1)
       "[1] Amplitude Min/Max"=<0.5 2>
      )
      Value=&Waveform(
       Global Amplitude Scale=2
       "Waveform [1]"=::NewEnumerator0
       "[1] Phase Scale"=$User.FlickerSpeed
       "[1] Frequency"=&RandomRangeFloat(Minimum=0.5, Maximum=1)
       "[1] Amplitude Min/Max"=<0.5 2>
      )
      Linear Color=$Particles.Initial.Color
     ))
    ScaleA=true
    ScaleRGB=true
   ) {module=ScaleColor001}
   &RegionMask(Outer Float=0.5, Region Falloff=1, Region Radius=3500, UseOffset=false, UseRotation=false)
   &ScaleColor(
    Scale Alpha=&FloatFromCurve(
     FloatCurve=DataInterface() {propertyValues {Curve {keys [
         {interpMode=RCIM_Cubic, tangentMode=RCTM_User, time=0, value=0}
         {arriveTangent=1.8182, leaveTangent=1.8182, time=0.35, value=1}
         <at 0.55 1>
         <at 1 0>
        ]}}}
     Scale Curve=&Multiply_Float(
      A=&Waveform(
       Global Amplitude Scale=$User.Alpha
       "Waveform [1]"=::NewEnumerator0
       "[1] Phase Scale"=$User.FlickerSpeed
       "[1] Frequency"=&RandomRangeFloat(Minimum=0.5, Maximum=1)
       "[1] Amplitude Min/Max"=<0.25 1>
      )
      B=$Particles.RegionMask.Float
     )
    )
    Scale RGB=&Multiply_VectorByFloat(Vector=<5 5 5>, Float=1)
    ScaleA=true
    ScaleRGB=true
   )
   &SolveForcesAndVelocity(Acceleration Limit=5, Clamp Velocity=true, Limit Acceleration=true, Speed Limit=100)
   &DynamicMaterialParameters(
    Index 0 Param 1=0.5
    Index 0 Param 2=1
    Index 0 Param 3=0.5
    Index 0 Param 4=0.5
    Index 1 Param 1=0.2
    Index 1 Param 2=0.2
    Index 1 Param 3=1
    Index 1 Param 4=&RandomRangeFloat(Minimum=0, Maximum=0.5)
    Index 2 Param 1=128
    Index 2 Param 3=&RandomRangeFloat(Minimum=0.1, Maximum=0.25)
    Index 2 Param 4=&RandomRangeFloat(Minimum=0.05, Maximum=0.25)
    Index 3 Param 1=2048
    Index 3 Param 3=1
    Index 3 Param 4=0.5
    Module.Index0Float0Write=true
    Module.Index0Float1Write=true
    Module.Index0Float2Write=true
    Module.Index0Float3Write=true
    Module.Index1Float0Write=true
    Module.Index1Float1Write=true
    Module.Index1Float2Write=true
    Module.Index1Float3Write=true
    Module.Index2Float0Write=true
    Module.Index2Float1Write=true
    Module.Index2Float2Write=true
    Module.Index2Float3Write=true
    Module.Index3Float0Write=true
    Module.Index3Float1Write=true
    Module.Index3Float2Write=true
    Module.Index3Float3Write=true
    Write Parameter Index 1=::NewEnumerator1
    Write Parameter Index 2=::NewEnumerator1
    Write Parameter Index 3=::NewEnumerator1
   )
   /SetVariables(
    Particles.NewColor:LinearColor=&MultiplyLinearColor(
     Linear Color=$Particles.Color
     Scale Factor=/Niagara/DynamicInputs/Multiply/Multiply_Vector4_ByFloat(
      Vector4:Vector4f=&MakeVector4(
       X=&RandomRangeFloat(Minimum=1, Maximum=1)
       Y=1
       Z=1
       W=&FloatFromCurve(
        FloatCurve=DataInterface() {propertyValues {Curve {keys [
            {arriveTangent=27.2697, interpMode=RCIM_Cubic, leaveTangent=27.2697, tangentMode=RCTM_User, time=0, value=0}
            {arriveTangent=-7.9125, interpMode=RCIM_Cubic, leaveTangent=-7.9125, tangentMode=RCTM_User, time=0.1, value=1}
            {tangentMode=RCTM_User, time=0.5, value=0.05}
           ]}}}
        Scale Curve=0.02
       )
      )
      Float:NiagaraFloat=&RandomRangeFloat(Minimum=1, Maximum=1)
     )
    )
    Particles.NewDynamic1:Vector4f=&MakeVector4(X=&RandomRangeFloat(Minimum=1, Maximum=1), Y=1, Z=0, W=0)
    Particles.NewSize:Vector2f=&Multiply_Vector2D_ByFloat(
     Float=&Lerp_Float(
      A=0
      B=5
      Alpha=&FloatFromCurve(
       FloatCurve=DataInterface() {propertyValues {Curve {keys [<at 0 0>,<at 0.7 0>,<at 1 1>]}}}
       CurveIndex=&RandomRangeFloat(Minimum=0, Maximum=1)
      )
     )
     Vector2D=$Particles.SpriteSize
    )
   ) {isSetParameters=true}
  ]
  rendererClasses [&NiagaraSpriteRendererProperties]
  renderers [
   {rendererClass=&NiagaraSpriteRendererProperties}
   {rendererClass=&NiagaraSpriteRendererProperties, rendererIndex=1}
  ]
 }
 Dust_Motes002 {
  bEnabled=false
  simTarget=GPUComputeSim
  emitterUpdateScript [
   /SetVariables(Emitter.CameraPostion:Vector3f=&ReturnCameraVector(Camera Property=::NewEnumerator0, Coordinate Space=::Simulation)) {isSetParameters=true, module=SetVariables_#3}
   &EmitterState(
    Enable Distance Culling=true
    Life Cycle Mode=::NewEnumerator1
    Loop Duration=&RandomRangeFloat(Minimum=1, Maximum=2)
    Max Distance Response=::SleepAndClearParticles
    MaxDistance=3000
    Module.bApplySpawnCountScaleByDistance=true
    Module.bUseMaxDistance=true
    Module.bUseMinDistance=false
    Recalculate Duration Each Loop=true
    Scalability Mode=::NewEnumerator1
    Scale Spawn Count=true
    Spawn Count Scale By Distance=DataInterface() {propertyValues {Curve {keys [<at 0 1>,<at 0.8 1>,<at 0.9962 0.4982>]}}}
    UseLoopDelay=false
   )
   &SpawnRate(SpawnRate=&Multiply_Float(A=1500, B=$User.SpawnRateMult), Use Spawn Probability=false)
  ]
  particleSpawnScript [
   /SetVariables(Particles.RandSpreadCurve:NiagaraFloat=&FloatFromCurve(
     FloatCurve=DataInterface() {propertyValues {Curve {keys [
         {interpMode=RCIM_Cubic, tangentMode=RCTM_User, time=0, value=0.7}
         {interpMode=RCIM_Cubic, tangentMode=RCTM_User, time=0.8, value=0.4}
         {arriveTangent=2.7919, interpMode=RCIM_Cubic, leaveTangent=2.7919, tangentMode=RCTM_User, time=1, value=1}
        ]}}}
     CurveIndex=&RandomRangeFloat(Minimum=0, Maximum=1)
    )) {isSetParameters=true, module=SetVariables_#4}
   &InitializeParticle(
    Color=&MakeLinearColorFromVectorAndFloat(
     "Vector (RGB)"=&MakeVectorFromLinearColorRGB(Color=&MultiplyLinearColor(
       Linear Color=&ColorFromCurve(
        ColorCurve=DataInterface() {propertyValues {
          AlphaCurve {keys []}
          BlueCurve {keys [<at 0.1189 1>,<at 0.1666 0>,<at 0.3109 0.4062>,<at 0.5746 1>,<at 0.8383 0>,<at 0.875 0>,<at 0.9354 1>,<at 1 1>]}
          GreenCurve {keys [<at 0.1189 0.5833>,<at 0.1666 0.1067>,<at 0.3109 0.9>,<at 0.5746 1>,<at 0.8383 0.5>,<at 0.875 0.2>,<at 0.9354 0.2942>,<at 1 0.625>]}
          RedCurve {keys [<at 0.1189 0.5>,<at 0.1666 1>,<at 0.3109 1>,<at 0.5746 1>,<at 0.8383 1>,<at 0.875 1>,<at 0.9354 0.23>,<at 1 0.5>]}
         }}
        CurveIndex=Linked() {linkedVariable {name=NPC.FortniteNPC.TimeOfDay_0-1, type=NiagaraFloat}}
       )
       Scale Factor=&MakeVector4_FromFloat(Float=1)
      ))
     "Float (Alpha)"=&RandomRangeFloat(Minimum=1, Maximum=1)
    )
    Lifetime=&UniformRangedFloat(Minimum=4, Maximum=5)
    Mass Mode=::NewEnumerator1
    Sprite Rotation Mode=::NewEnumerator2
    Sprite Size=&Vector2DFromFloat(Value=&RandomRangeFloat(Minimum=1, Maximum=2))
    Sprite Size Mode=::NewEnumerator4
    UsePositionOffset=false
    Write Material Random=true
   ) {enabled=false, module=InitializeParticle001}
   &InitializeParticle(
    Color Maximum=$System.ColorNear
    Color Minimum=$System.ColorFar
    Color Mode=::NewEnumerator2
    Lifetime=&Multiply_Float(A=&UniformRangedFloat(Minimum=0.5, Maximum=1), B=$User.LifeTimeMult)
    Mass Mode=::NewEnumerator1
    Sprite Rotation Mode=::NewEnumerator2
    Sprite Size=&Vector2DFromFloat(Value=&Multiply_Float(A=&RandomRangeFloat(Minimum=10, Maximum=15), B=$User.Size))
    Sprite Size Mode=::NewEnumerator4
    UsePositionOffset=false
    Write Material Random=true
   )
   &FN_SetSpawnGroup() {enabled=false}
   &ShapeLocation(MaskSpawnGroup=true, Override Seed=false, Rotation Mode=::NewEnumerator1, Sphere Radius=2500, UseOwnerScale=true)
   &KillParticlesInVolume(Box Size:Vector3f=<5000 5000 800>, Kill Shape=::NewEnumerator1) {enabled=false}
   &AddVelocity(Velocity=&RandomVector(Vector Scale=&RandomRangeFloat(Minimum=0.5, Maximum=1)))
   &MaintainInCameraParticleScale()
   &DynamicMaterialParameters(
    Index 0 Param 1=2
    Index 0 Param 2=50
    Index 0 Param 3=200
    Module.Param0WriteEnabled=true
    Module.Param1WriteEnabled=true
    Module.Param2WriteEnabled=true
    Module.Param3WriteEnabled=true
   ) {module=DynamicMaterialParameters001}
   &ShapeLocation(
    MaskSpawnGroup=true
    Override Seed=false
    Spawn Group Mask:NiagaraInt32=&RandomRangeInt(Minimum=0, Maximum=3)
    Sphere Radius=150
    UseOwnerScale=true
   ) {enabled=false, module=ShapeLocation003}
  ]
  particleUpdateScript [
   &ParticleState()
   &ScaleSpriteSize(
    Uniform Curve Scale=&FN_LOD_Distance_Remap(Start Distance=200, End Distance=3000, Minimum Output=1, Maximum Output=0.25, Clamp Output=true)
    Uniform Curve Sprite Scale=DataInterface() {propertyValues {Curve {keys [
        {interpMode=RCIM_Cubic, tangentMode=RCTM_User, time=0, value=0}
        <at 0.2 1>
        {arriveTangent=-1.3125, leaveTangent=-1.3125, time=0.7, value=1}
        {tangentMode=RCTM_User, time=1, value=0}
       ]}}}
   ) {module=ScaleSpriteSize001}
   &self:Camera_to_Particle_Distance_Calculation(Camera Postion:Vector3f=$Emitter.CameraPostion, Distance=1500, Particle Position:NiagaraPosition=$Particles.Position) {module=Camera_to_Particle_Distance_Calculation}
   &ScaleSpriteSize(Scale Factor=&Vector2DFromFloat(Value=&FloatFromCurve(
      FloatCurve=DataInterface() {propertyValues {Curve {keys [<at 0 0.25 RCIM_Cubic>,<at 1 1.5 RCIM_Cubic>]}}}
      CurveIndex=$Particles.NormalizedDistance
     )))
   &GravityForce(Gravity=&MakeVector(
     X=0
     Y=0
     Z=&FloatFromCurve(
      FloatCurve=DataInterface() {propertyValues {Curve {keys [<at 0 1 RCIM_Cubic>,<at 1 0>]}}}
      Scale Curve=&RandomRangeFloat(Minimum=-25, Maximum=-50)
     )
    ))
   &Drag(Drag=0.5, Use Linear Drag=true, Use Rotational Drag=false)
   &CurlNoiseForce(
    Noise Frequency=&RandomRangeFloat(Minimum=0.2, Maximum=1)
    Noise Strength=&RandomRangeFloat(Minimum=15, Maximum=50)
    Pan Noise=true
    Pan Noise Field=<0 0 1>
   )
   &SpriteRotationRate(Rotation Rate=&RandomRangeFloat(Minimum=-45, Maximum=45))
   &ScaleColor(
    Scale RGB=&MakeVectorFromLinearColorRGB(Color=&HueShiftLinearColor(
      Hue=0
      Saturation=&Waveform(
       Global Amplitude Scale=5
       "Waveform [1]"=::NewEnumerator0
       "[1] Frequency"=&RandomRangeFloat(Minimum=0.5, Maximum=1)
       "[1] Amplitude Min/Max"=<1 2>
      )
      Value=&Waveform(
       Global Amplitude Scale=2
       "Waveform [1]"=::NewEnumerator0
       "[1] Phase Scale"=$User.FlickerSpeed
       "[1] Frequency"=&RandomRangeFloat(Minimum=0.5, Maximum=1)
       "[1] Amplitude Min/Max"=<0.5 2>
      )
      Linear Color=$Particles.Initial.Color
     ))
    ScaleA=true
    ScaleRGB=true
   ) {module=ScaleColor001}
   &RegionMask(Outer Float=0.5, Region Falloff=1, Region Radius=3500, UseOffset=false, UseRotation=false)
   &ScaleColor(
    Scale Alpha=&FloatFromCurve(
     FloatCurve=DataInterface() {propertyValues {Curve {keys [
         {interpMode=RCIM_Cubic, tangentMode=RCTM_User, time=0, value=0}
         {arriveTangent=1.8182, leaveTangent=1.8182, time=0.35, value=1}
         <at 0.55 1>
         <at 1 0>
        ]}}}
     Scale Curve=&Multiply_Float(
      A=&Waveform(
       Global Amplitude Scale=$User.Alpha
       "Waveform [1]"=::NewEnumerator0
       "[1] Phase Scale"=$User.FlickerSpeed
       "[1] Frequency"=&RandomRangeFloat(Minimum=0.5, Maximum=1)
       "[1] Amplitude Min/Max"=<0.25 1>
      )
      B=$Particles.RegionMask.Float
     )
    )
    Scale RGB=&Multiply_VectorByFloat(Vector=<1 1 1>, Float=0.25)
    ScaleA=true
    ScaleRGB=true
   )
   &SolveForcesAndVelocity(Acceleration Limit=5, Clamp Velocity=true, Limit Acceleration=true, Speed Limit=100)
   &DynamicMaterialParameters(
    Index 0 Param 1=0.5
    Index 0 Param 2=1
    Index 0 Param 4=0.5
    Module.Index0Float0Write=true
    Module.Index0Float1Write=true
    Module.Index0Float2Write=true
    Module.Index0Float3Write=true
   )
  ]
  rendererClasses [&NiagaraSpriteRendererProperties]
  renderers [{rendererClass=&NiagaraSpriteRendererProperties}]
 }
}
summary {
 systemName=NS_Env_HermesGold_Gold_DigSite_InAir
 emitters [
  {bEnabled=false, emitterName=DEBUG, rendererClasses [&NiagaraSpriteRendererProperties]}
  {bEnabled=false, emitterName=Dust_Motes, simTarget=GPUComputeSim, rendererClasses [&NiagaraSpriteRendererProperties]}
  {bEnabled=false, emitterName=Dust_Glints, simTarget=GPUComputeSim, rendererClasses [&NiagaraSpriteRendererProperties]}
  {emitterName=Dust_Motes001, simTarget=GPUComputeSim, rendererClasses [&NiagaraSpriteRendererProperties]}
  {bEnabled=false, emitterName=Dust_Motes002, simTarget=GPUComputeSim, rendererClasses [&NiagaraSpriteRendererProperties]}
 ]
 userVariables [
  User.Alpha {type=NiagaraFloat, defaultValue=1.5, description=""}
  User.FlickerSpeed {type=NiagaraFloat, defaultValue=0.5, description=""}
  User.LifeTimeMult {type=NiagaraFloat, defaultValue=0.75, description=""}
  User.Size {type=NiagaraFloat, defaultValue=1.25, description=""}
  User.SpawnRateMult {type=NiagaraFloat, defaultValue=1, description=""}
 ]
}
