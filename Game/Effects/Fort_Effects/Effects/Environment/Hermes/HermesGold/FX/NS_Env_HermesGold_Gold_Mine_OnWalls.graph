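# GPU Niagara system for gold glitter sprites on mine surfaces in the HermesGold environment; all four emitters are currently disabled.
# Two glitter emitters and Parent snap spawns to the depth buffer, filtered by GBuffer hue.
# Child spawns trailing sparkles from Parent particles through a particle attribute reader.
# Curl noise drives motion, gravity pulls Parent and Child down, camera distance scales size, and waveforms flicker glitter and Child color.
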NiagaraSystem /Game/Effects/Fort_Effects/Effects/Environment/Hermes/HermesGold/FX/NS_Env_HermesGold_Gold_Mine_OnWalls
dependencies {
 usedDataInterfaces [
  &NiagaraDataInterfaceCamera
  &NiagaraDataInterfaceGBuffer
  &NiagaraDataInterfaceCurve
  &NiagaraDataInterfaceParticleRead
 ]
 usedDynamicInputs []
 usedModules [
  &SystemState
  &self:SystemSpawnScript.SystemScriptSource.SystemScriptGraph.NiagaraNodeAssignment_0.SetVariables_#1
  &self:SystemSpawnScript.SystemScriptSource.SystemScriptGraph.NiagaraNodeAssignment_1.SetVariables_#2
 ]
}
emitters {
 Child {
  bEnabled=false
  simTarget=GPUComputeSim
  emitterUpdateScript [
   /SetVariables(Emitter.CameraPostion:Vector3f=&ReturnCameraVector(Camera Property=::NewEnumerator0, Coordinate Space=::Simulation)) {isSetParameters=true, module=SetVariables_#3}
   &EmitterState()
   &SpawnParticlesFromOtherEmitter(
    Attribute Reader=DataInterface() {propertyValues {EmitterBinding {bindingMode=Other, emitterName=Parent}}}
    Maximum Age Spawn Threshold=1.5
    Minimum Age Spawn Threshold=0.1
    Spawn Rate=100
    Spawn Rate Per Particle Cap=50
    Use Maximum Age=true
    Use Minimum Age=true
   )
  ]
  particleSpawnScript [
   &InitializeParticle(
    Color=<rgba 1 0.516 0.0482 1>
    Lifetime Max=5
    Lifetime Min=4
    Lifetime Mode=::NewEnumerator1
    Sprite Size Mode=::NewEnumerator1
    Sprite UV Mode=::NewEnumerator1
    Uniform Sprite Size Max=25
    Uniform Sprite Size Min=2
    UsePositionOffset=false
    Write Material Random=true
   )
   &SampleParticlesFromOtherEmitter(
    Applied Velocity Scale:NiagaraFloat=&RandomRangeFloat(Minimum=0.1, Maximum=0.2)
    Velocity Sampling=::NewEnumerator0
   )
   &ShapeLocation(
    MaskSpawnGroup=false
    Offset Mode=::NewEnumerator0
    Override Seed=false
    Sphere Radius=&FloatFromCurve(FloatCurve=DataInterface() {propertyValues {Curve {keys [<at 0 0.2>,<at 0.5 5>]}}}, CurveIndex=$System.NormalizedLoopAge, Scale Curve=25)
    UseOwnerScale=true
   )
   &CameraOffset(Camera Offset Amount=15)
   &CameraOffset(Camera Offset Amount=25) {module=CameraOffset002}
  ]
  particleUpdateScript [
   &ParticleState()
   &self:Camera_to_Particle_Distance_Calculation(Camera Postion:Vector3f=$Emitter.CameraPostion, Distance=3500, Offset=-500, Particle Position:NiagaraPosition=$Particles.Position) {module=Camera_to_Particle_Distance_Calculation}
   &ScaleSpriteSize(Scale Factor=&Vector2DFromFloat(Value=&FloatFromCurve(
      FloatCurve=DataInterface() {propertyValues {Curve {keys [<at 0.25 0 RCIM_Cubic>,<at 1 1.5 RCIM_Cubic>]}}}
      CurveIndex=$Particles.NormalizedDistance
     ))) {module=ScaleSpriteSize001}
   &CurlNoiseForce(
    Curl Noise Cone Mask Angle=80
    Curl Noise Cone Mask Falloff Angle=35
    Mask Curl Noise=true
    Noise Frequency=&RandomRangeFloat(Minimum=0.25, Maximum=2)
    Noise Strength=&Multiply_Float(
     A=&RandomRangeFloat(Minimum=100, Maximum=250)
     B=&FloatFromCurve(FloatCurve=DataInterface() {propertyValues {Curve {keys [
          {interpMode=RCIM_Cubic, tangentMode=RCTM_User, time=0.25, value=1}
          {interpMode=RCIM_Cubic, tangentMode=RCTM_User, time=0.5, value=1}
          {tangentMode=RCTM_User, time=1, value=0}
         ]}}})
    )
    Pan Noise=false
   )
   /Niagara/Modules/Math/AvoidDistanceFieldSurfaces_GPU(
    Avoid Nearest Surface=true
    Avoid Oncoming Surfaces=true
    Falloff Distance=1
    Max Trace Distance=50
    Nearest Surface Avoidance Strength=150
    Obstactle Avoidance Strength=25
    Trace Vector:Vector3f=&Multiply_Vector(A=$Particles.Velocity, B=<1 1 0>)
   ) {enabled=false}
   &GravityForce(Gravity=<0 0 -250>) {module=GravityForce001}
   &VectorNoiseForce(
    Force Amount=&FloatFromCurve(FloatCurve=DataInterface() {propertyValues {Curve {keys [<at 0 0>,<at 1 1>]}}}, Scale Curve=2)
    Module.Use Random Noise Time Interval=false
   )
   &Drag(Drag=2, Use Linear Drag=true, Use Rotational Drag=true)
   &SolveForcesAndVelocity(
    Acceleration Limit=20
    Clamp Velocity=true
    Force:Vector3f=&Multiply_Vector(A=$Transient.PhysicsForce, B=<1 1 1>)
    Limit Acceleration=true
    Speed Limit=250
    Velocity:Vector3f=&Multiply_Vector(A=$Particles.Velocity, B=<1 1 1>)
   )
   &CalculateAccurateVelocity() {enabled=false}
   &DynamicMaterialParameters(
    Index 0 Param 1=0.2
    Index 0 Param 2=&RandomRangeFloat(Minimum=0.2, Maximum=0.4)
    Index 0 Param 3=1
    Module.Index0Param0Write=true
    Module.Index0Param1Write=true
    Module.Index0Param2Write=true
    Module.Index0Param3Write=true
   )
   &ScaleColor(
    Scale Alpha=&FloatFromCurve(
     FloatCurve=DataInterface() {propertyValues {Curve {keys [
         {interpMode=RCIM_Cubic, tangentMode=RCTM_User, time=0, value=0}
         {interpMode=RCIM_Cubic, tangentMode=RCTM_User, time=0.2, value=0.8}
         <at 1 0>
        ]}}}
     Scale Curve=0.25
    )
    ScaleA=true
    ScaleRGB=true
   )
   &ScaleSpriteSize(
    Uniform Curve Scale=3
    Uniform Curve Sprite Scale=DataInterface() {propertyValues {Curve {keys [
        {interpMode=RCIM_Cubic, tangentMode=RCTM_User, time=0, value=0}
        {interpMode=RCIM_Cubic, tangentMode=RCTM_User, time=0.2, value=1}
        {tangentMode=RCTM_User, time=1, value=0}
       ]}}}
   )
   &CameraOffset(Camera Offset Amount=50) {module=CameraOffset001}
   /SetVariables(
    Particles.Linear Color=&Multiply_LinearColorByFloat(
     Linear Color=$Particles.Color
     Float=&Waveform(
      Global Amplitude Scale=5
      "Waveform [1]"=::NewEnumerator0
      "[1] Frequency"=&RandomRangeFloat(Minimum=0.5, Maximum=2)
      "[1] Amplitude Min/Max"=<0 1>
     )
    )
    Particles.Vector 2D:Vector2f=&Multiply_Vector2D(A=$Particles.SpriteSize, B=<0.25 0.25>)
   ) {isSetParameters=true, module=SetVariables_#4}
  ]
  rendererClasses [&NiagaraSpriteRendererProperties]
  renderers [
   {rendererClass=&NiagaraSpriteRendererProperties}
   {rendererClass=&NiagaraSpriteRendererProperties, rendererIndex=1}
  ]
 }
 HermesGold_Glitter {
  bEnabled=false
  simTarget=GPUComputeSim
  emitterUpdateScript [
   /SetVariables(Emitter.CameraPostion:Vector3f=&ReturnCameraVector(Camera Property=::NewEnumerator0, Coordinate Space=::Simulation)) {isSetParameters=true, module=SetVariables_#3}
   &EmitterState()
   &SpawnRate(SpawnRate=5000, Use Spawn Probability=false)
  ]
  particleSpawnScript [
   &InitializeParticle(
    Color=<rgba 1 0.516 0.0482 1>
    Lifetime Max=3
    Lifetime Min=2
    Lifetime Mode=::NewEnumerator1
    Sprite Size Mode=::NewEnumerator1
    Sprite UV Mode=::NewEnumerator1
    Uniform Sprite Size Max=50
    Uniform Sprite Size Min=2
    UsePositionOffset=false
    Write Material Random=true
   )
   &ShapeLocation(MaskSpawnGroup=false, Offset Mode=::NewEnumerator0, Override Seed=false, Sphere Radius=1500, UseOwnerScale=true)
   /SetVariables(Particles.Position=&Add_Vector(
     A=Linked() {linkedVariable {name=Particles.Position, type=NiagaraPosition}}
     B=&MakeVector(X=0, Y=0, Z=&Waveform("Waveform [1]"=::NewEnumerator0, "[1] Frequency"=5, "[1] Amplitude Min/Max"=<0 25>))
    )) {isSetParameters=true, module=SetVariables_#5}
   /SetVariables(Particles.New Vector 3f:Vector3f=<0 0 1>) {isSetParameters=true, module=SetVariables_#6}
   &CameraOffset(Camera Offset Amount=15)
   /SetVariables() {isSetParameters=true, module=SetVariables_#7}
   /SetVariables(Particles.InitialPos:NiagaraPosition=$Particles.Position) {isSetParameters=true, module=SetVariables_#8}
   /Niagara/Modules/Math/PlaceParticlesOnDepthBuffer_GPU(Max Z Depth=8000, Module.Use Max Z Depth=true)
   &KillParticles(Kill Particles=&SetBoolByFloatComparison(
     A=&VectorLength(Vector=&Subtract_Vector(A=$Particles.InitialPos, B=Linked() {linkedVariable {name=Particles.Position, type=NiagaraPosition}}))
     B=5000
    )) {module=KillParticles001}
   &KillParticles(Kill Particles=&Boolean_NotOperation(Bool=$Output.PlaceParticlesOnDepthBuffer_GPU.ValidSample)) {module=KillParticles002}
   /Game/Effects/Niagara/Modules/Spawn/FN_Spawn_GBuffer_Color_WithinRangeV2(HueMax=25, HueMin=23)
   &self:ScratchModule_07(Angle=180, BlendDirection=1, Normal:Vector3f=<0 0 1>, PickCeilingOrFloor=true) {enabled=false, module=ScratchModule_07}
   &ApplyInitialForces()
   &CameraOffset(Camera Offset Amount=25) {module=CameraOffset002}
  ]
  particleUpdateScript [
   &ParticleState()
   &self:Camera_to_Particle_Distance_Calculation(Camera Postion:Vector3f=$Emitter.CameraPostion, Distance=3500, Offset=-1500, Particle Position:NiagaraPosition=$Particles.Position) {module=Camera_to_Particle_Distance_Calculation}
   &ScaleSpriteSize(Scale Factor=&Vector2DFromFloat(Value=&FloatFromCurve(
      FloatCurve=DataInterface() {propertyValues {Curve {keys [<at 0.25 0 RCIM_Cubic>,<at 1 1.5 RCIM_Cubic>]}}}
      CurveIndex=$Particles.NormalizedDistance
     ))) {module=ScaleSpriteSize001}
   &CurlNoiseForce(
    Curl Noise Cone Mask Angle=25
    Mask Curl Noise=true
    Noise Frequency=&RandomRangeFloat(Minimum=2, Maximum=4)
    Noise Strength=&Multiply_Float(A=&RandomRangeFloat(Minimum=2500, Maximum=2500), B=1)
    Pan Noise=true
   )
   &AvoidDistanceFieldSurfaces_GPU(
    Avoid Nearest Surface=true
    Avoid Oncoming Surfaces=true
    Falloff Distance=1
    Max Trace Distance=50
    Nearest Surface Avoidance Strength=150
    Obstactle Avoidance Strength=25
    Trace Vector:Vector3f=&Multiply_Vector(A=$Particles.Velocity, B=<1 1 0>)
   ) {enabled=false}
   &GravityForce(Gravity=<0 0 0>)
   &SolveForcesAndVelocity(
    Acceleration Limit=250
    Clamp Velocity=true
    Force:Vector3f=&Multiply_Vector(A=$Transient.PhysicsForce, B=<1 1 1>)
    Limit Acceleration=true
    Speed Limit=250
    Velocity:Vector3f=&Multiply_Vector(A=$Particles.Velocity, B=<1 1 1>)
   )
   &CalculateAccurateVelocity() {enabled=false}
   &DynamicMaterialParameters(
    Index 0 Param 1=0.2
    Index 0 Param 2=&RandomRangeFloat(Minimum=0.2, Maximum=0.4)
    Index 0 Param 3=1
    Module.Index0Param0Write=true
    Module.Index0Param1Write=true
    Module.Index0Param2Write=true
    Module.Index0Param3Write=true
   )
   &ScaleColor(
    Scale Alpha=&FloatFromCurve(
     FloatCurve=DataInterface() {propertyValues {Curve {keys [
         {interpMode=RCIM_Cubic, tangentMode=RCTM_User, time=0, value=0}
         {interpMode=RCIM_Cubic, tangentMode=RCTM_User, time=0.2, value=0.8}
         <at 1 0>
        ]}}}
     Scale Curve=0.5
    )
    ScaleA=true
    ScaleRGB=true
   )
   &ScaleSpriteSize(
    Uniform Curve Scale=3
    Uniform Curve Sprite Scale=DataInterface() {propertyValues {Curve {keys [
        {interpMode=RCIM_Cubic, tangentMode=RCTM_User, time=0, value=0}
        {interpMode=RCIM_Cubic, tangentMode=RCTM_User, time=0.2, value=1}
        {tangentMode=RCTM_User, time=1, value=0}
       ]}}}
   )
   &CameraOffset(Camera Offset Amount=50) {module=CameraOffset001}
   /SetVariables(
    Particles.Linear Color=&Multiply_LinearColorByFloat(
     Linear Color=$Particles.Color
     Float=&Waveform(
      Global Amplitude Scale=5
      "Waveform [1]"=::NewEnumerator0
      "[1] Frequency"=&RandomRangeFloat(Minimum=0.5, Maximum=2)
      "[1] Amplitude Min/Max"=<0 1>
     )
    )
    Particles.Vector 2D:Vector2f=&Multiply_Vector2D(A=$Particles.SpriteSize, B=<0.25 0.25>)
   ) {isSetParameters=true, module=SetVariables_#4}
  ]
  rendererClasses [&NiagaraSpriteRendererProperties]
  renderers [
   {rendererClass=&NiagaraSpriteRendererProperties}
   {rendererClass=&NiagaraSpriteRendererProperties, rendererIndex=1}
  ]
 }
 HermesGold_Glitter001 {
  bEnabled=false
  simTarget=GPUComputeSim
  emitterUpdateScript [
   /SetVariables(Emitter.CameraPostion:Vector3f=&ReturnCameraVector(Camera Property=::NewEnumerator0, Coordinate Space=::Simulation)) {isSetParameters=true, module=SetVariables_#3}
   &EmitterState()
   &SpawnRate(SpawnRate=100, Use Spawn Probability=false)
  ]
  particleSpawnScript [
   &InitializeParticle(
    Color=<rgba 1 0.516 0.0482 1>
    Lifetime Max=5
    Lifetime Min=4
    Lifetime Mode=::NewEnumerator1
    Sprite Size Mode=::NewEnumerator1
    Sprite UV Mode=::NewEnumerator1
    Uniform Sprite Size Max=50
    Uniform Sprite Size Min=2
    UsePositionOffset=false
    Write Material Random=true
   )
   &ShapeLocation(MaskSpawnGroup=false, Offset Mode=::NewEnumerator0, Override Seed=false, Sphere Radius=1500, UseOwnerScale=true)
   /SetVariables(Particles.Position=&Add_Vector(
     A=Linked() {linkedVariable {name=Particles.Position, type=NiagaraPosition}}
     B=&MakeVector(X=0, Y=0, Z=&Waveform("Waveform [1]"=::NewEnumerator0, "[1] Frequency"=5, "[1] Amplitude Min/Max"=<0 25>))
    )) {isSetParameters=true, module=SetVariables_#5}
   /SetVariables(Particles.New Vector 3f:Vector3f=<0 0 1>) {isSetParameters=true, module=SetVariables_#6}
   &CameraOffset(Camera Offset Amount=15)
   /SetVariables() {isSetParameters=true, module=SetVariables_#7}
   /SetVariables(Particles.InitialPos:NiagaraPosition=$Particles.Position) {isSetParameters=true, module=SetVariables_#8}
   &PlaceParticlesOnDepthBuffer_GPU(Max Z Depth=8000, Module.Use Max Z Depth=true)
   &KillParticles(Kill Particles=&SetBoolByFloatComparison(
     A=&VectorLength(Vector=&Subtract_Vector(A=$Particles.InitialPos, B=Linked() {linkedVariable {name=Particles.Position, type=NiagaraPosition}}))
     B=5000
    )) {module=KillParticles001}
   &KillParticles(Kill Particles=&Boolean_NotOperation(Bool=$Output.PlaceParticlesOnDepthBuffer_GPU.ValidSample)) {module=KillParticles002}
   &FN_Spawn_GBuffer_Color_WithinRangeV2(HueMax=25, HueMin=23)
   &self:ScratchModule_07(Angle=180, BlendDirection=1, Normal:Vector3f=<0 0 1>, PickCeilingOrFloor=true) {module=ScratchModule_07}
   &ApplyInitialForces()
   &CameraOffset(Camera Offset Amount=25) {module=CameraOffset002}
  ]
  particleUpdateScript [
   &ParticleState()
   &self:Camera_to_Particle_Distance_Calculation(Camera Postion:Vector3f=$Emitter.CameraPostion, Distance=3500, Offset=-500, Particle Position:NiagaraPosition=$Particles.Position) {module=Camera_to_Particle_Distance_Calculation}
   &ScaleSpriteSize(Scale Factor=&Vector2DFromFloat(Value=&FloatFromCurve(
      FloatCurve=DataInterface() {propertyValues {Curve {keys [<at 0.25 0 RCIM_Cubic>,<at 1 1.5 RCIM_Cubic>]}}}
      CurveIndex=$Particles.NormalizedDistance
     ))) {module=ScaleSpriteSize001}
   &CurlNoiseForce(
    Curl Noise Cone Mask Angle=25
    Mask Curl Noise=true
    Noise Frequency=&RandomRangeFloat(Minimum=2, Maximum=4)
    Noise Strength=&Multiply_Float(A=&RandomRangeFloat(Minimum=2500, Maximum=2500), B=1)
    Pan Noise=true
   )
   &AvoidDistanceFieldSurfaces_GPU(
    Avoid Nearest Surface=true
    Avoid Oncoming Surfaces=true
    Falloff Distance=1
    Max Trace Distance=50
    Nearest Surface Avoidance Strength=150
    Obstactle Avoidance Strength=25
    Trace Vector:Vector3f=&Multiply_Vector(A=$Particles.Velocity, B=<1 1 0>)
   ) {enabled=false}
   &GravityForce(Gravity=<0 0 0>)
   &SolveForcesAndVelocity(
    Acceleration Limit=250
    Clamp Velocity=true
    Force:Vector3f=&Multiply_Vector(A=$Transient.PhysicsForce, B=<1 1 1>)
    Limit Acceleration=true
    Speed Limit=250
    Velocity:Vector3f=&Multiply_Vector(A=$Particles.Velocity, B=<1 1 1>)
   )
   &CalculateAccurateVelocity() {enabled=false}
   &DynamicMaterialParameters(
    Index 0 Param 1=0.2
    Index 0 Param 2=&RandomRangeFloat(Minimum=0.2, Maximum=0.4)
    Index 0 Param 3=1
    Module.Index0Param0Write=true
    Module.Index0Param1Write=true
    Module.Index0Param2Write=true
    Module.Index0Param3Write=true
   )
   &ScaleColor(
    Scale Alpha=&FloatFromCurve(
     FloatCurve=DataInterface() {propertyValues {Curve {keys [
         {interpMode=RCIM_Cubic, tangentMode=RCTM_User, time=0, value=0}
         {interpMode=RCIM_Cubic, tangentMode=RCTM_User, time=0.2, value=0.8}
         <at 1 0>
        ]}}}
     Scale Curve=0.25
    )
    ScaleA=true
    ScaleRGB=true
   )
   &ScaleSpriteSize(
    Uniform Curve Scale=3
    Uniform Curve Sprite Scale=DataInterface() {propertyValues {Curve {keys [
        {interpMode=RCIM_Cubic, tangentMode=RCTM_User, time=0, value=0}
        {interpMode=RCIM_Cubic, tangentMode=RCTM_User, time=0.2, value=1}
        {tangentMode=RCTM_User, time=1, value=0}
       ]}}}
   )
   &CameraOffset(Camera Offset Amount=50) {module=CameraOffset001}
   /SetVariables(
    Particles.Linear Color=&Multiply_LinearColorByFloat(
     Linear Color=$Particles.Color
     Float=&Waveform(
      Global Amplitude Scale=5
      "Waveform [1]"=::NewEnumerator0
      "[1] Frequency"=&RandomRangeFloat(Minimum=0.5, Maximum=2)
      "[1] Amplitude Min/Max"=<0 1>
     )
    )
    Particles.Vector 2D:Vector2f=&Multiply_Vector2D(A=$Particles.SpriteSize, B=<0.25 0.25>)
   ) {isSetParameters=true, module=SetVariables_#4}
  ]
  rendererClasses [&NiagaraSpriteRendererProperties]
  renderers [
   {rendererClass=&NiagaraSpriteRendererProperties}
   {rendererClass=&NiagaraSpriteRendererProperties, rendererIndex=1}
  ]
 }
 Parent {
  bEnabled=false
  simTarget=GPUComputeSim
  emitterUpdateScript [
   /SetVariables(Emitter.CameraPostion:Vector3f=&ReturnCameraVector(Camera Property=::NewEnumerator0, Coordinate Space=::Simulation)) {isSetParameters=true, module=SetVariables_#3}
   &EmitterState()
   &SpawnRate(
    SpawnRate=&FloatFromCurve(
     FloatCurve=DataInterface() {propertyValues {Curve {keys [<at 0 1>,<at 1 1>]}}}
     CurveIndex=$Emitter.NormalizedLoopAge
     Scale Curve=&RandomRangeFloat(Minimum=5, Maximum=5)
    )
    Use Spawn Probability=false
   )
  ]
  particleSpawnScript [
   &InitializeParticle(
    Color=<rgba 1 0.516 0.0482 1>
    Lifetime Max=3
    Lifetime Min=2
    Lifetime Mode=::NewEnumerator1
    Sprite Size Mode=::NewEnumerator1
    Sprite UV Mode=::NewEnumerator1
    Uniform Sprite Size Max=50
    Uniform Sprite Size Min=2
    UsePositionOffset=false
    Write Material Random=true
   )
   &ShapeLocation(MaskSpawnGroup=false, Offset Mode=::NewEnumerator0, Override Seed=false, Sphere Radius=1500, UseOwnerScale=true)
   /SetVariables(Particles.Position=&Add_Vector(
     A=Linked() {linkedVariable {name=Particles.Position, type=NiagaraPosition}}
     B=&MakeVector(X=0, Y=0, Z=&Waveform("Waveform [1]"=::NewEnumerator0, "[1] Frequency"=5, "[1] Amplitude Min/Max"=<0 25>))
    )) {isSetParameters=true, module=SetVariables_#5}
   /SetVariables(Particles.New Vector 3f:Vector3f=<0 0 1>) {isSetParameters=true, module=SetVariables_#6}
   &CameraOffset(Camera Offset Amount=15)
   /SetVariables() {isSetParameters=true, module=SetVariables_#7}
   /SetVariables(Particles.InitialPos:NiagaraPosition=$Particles.Position) {isSetParameters=true, module=SetVariables_#8}
   &PlaceParticlesOnDepthBuffer_GPU(Max Z Depth=8000, Module.Use Max Z Depth=true)
   &KillParticles(Kill Particles=&SetBoolByFloatComparison(
     A=&VectorLength(Vector=&Subtract_Vector(A=$Particles.InitialPos, B=Linked() {linkedVariable {name=Particles.Position, type=NiagaraPosition}}))
     B=5000
    )) {module=KillParticles001}
   &KillParticles(Kill Particles=&Boolean_NotOperation(Bool=$Output.PlaceParticlesOnDepthBuffer_GPU.ValidSample)) {module=KillParticles002}
   &FN_Spawn_GBuffer_Color_WithinRangeV2(HueMax=25, HueMin=23)
   &self:ScratchModule_07(Angle=180, BlendDirection=1, Normal:Vector3f=<0 0 1>, PickCeilingOrFloor=true) {module=ScratchModule_07}
   &ApplyInitialForces()
   &CameraOffset(Camera Offset Amount=25) {module=CameraOffset002}
  ]
  particleUpdateScript [
   &ParticleState()
   &self:Camera_to_Particle_Distance_Calculation(Camera Postion:Vector3f=$Emitter.CameraPostion, Distance=3500, Offset=-500, Particle Position:NiagaraPosition=$Particles.Position) {module=Camera_to_Particle_Distance_Calculation}
   &ScaleSpriteSize(Scale Factor=&Vector2DFromFloat(Value=&FloatFromCurve(
      FloatCurve=DataInterface() {propertyValues {Curve {keys [<at 0.25 0 RCIM_Cubic>,<at 1 1.5 RCIM_Cubic>]}}}
      CurveIndex=$Particles.NormalizedDistance
     ))) {module=ScaleSpriteSize001}
   &CurlNoiseForce(
    Curl Noise Cone Mask Angle=80
    Curl Noise Cone Mask Falloff Angle=15
    Mask Curl Noise=true
    Noise Frequency=&RandomRangeFloat(Minimum=0.5, Maximum=1)
    Noise Strength=&Multiply_Float(A=&RandomRangeFloat(Minimum=800, Maximum=2500), B=1)
    Pan Noise=true
   )
   &AvoidDistanceFieldSurfaces_GPU(
    Avoid Nearest Surface=true
    Avoid Oncoming Surfaces=true
    Falloff Distance=1
    Max Trace Distance=50
    Nearest Surface Avoidance Strength=150
    Obstactle Avoidance Strength=25
    Trace Vector:Vector3f=&Multiply_Vector(A=$Particles.Velocity, B=<1 1 0>)
   ) {enabled=false}
   &GravityForce(Gravity=<0 0 -500>) {module=GravityForce001}
   &Drag(Use Linear Drag=true, Use Rotational Drag=true)
   &SolveForcesAndVelocity(
    Acceleration Limit=500
    Clamp Velocity=true
    Force:Vector3f=&Multiply_Vector(A=$Transient.PhysicsForce, B=<1 1 1>)
    Limit Acceleration=true
    Speed Limit=500
    Velocity:Vector3f=&Multiply_Vector(A=$Particles.Velocity, B=<1 1 1>)
   )
   &CalculateAccurateVelocity() {enabled=false}
   &DynamicMaterialParameters(
    Index 0 Param 1=0.2
    Index 0 Param 2=&RandomRangeFloat(Minimum=0.2, Maximum=0.4)
    Index 0 Param 3=1
    Module.Index0Param0Write=true
    Module.Index0Param1Write=true
    Module.Index0Param2Write=true
    Module.Index0Param3Write=true
   )
   &ScaleColor(
    Scale Alpha=&FloatFromCurve(
     FloatCurve=DataInterface() {propertyValues {Curve {keys [
         {interpMode=RCIM_Cubic, tangentMode=RCTM_User, time=0, value=0}
         {interpMode=RCIM_Cubic, tangentMode=RCTM_User, time=0.2, value=0.8}
         <at 1 0>
        ]}}}
     Scale Curve=0.25
    )
    ScaleA=true
    ScaleRGB=true
   )
   &ScaleSpriteSize(
    Uniform Curve Scale=3
    Uniform Curve Sprite Scale=DataInterface() {propertyValues {Curve {keys [
        {interpMode=RCIM_Cubic, tangentMode=RCTM_User, time=0, value=0}
        {interpMode=RCIM_Cubic, tangentMode=RCTM_User, time=0.2, value=1}
        {tangentMode=RCTM_User, time=1, value=0}
       ]}}}
   )
  ]
  rendererClasses [&NiagaraSpriteRendererProperties]
  renderers [
   {rendererClass=&NiagaraSpriteRendererProperties}
   {rendererClass=&NiagaraSpriteRendererProperties, rendererIndex=1}
  ]
 }
}
summary {
 systemName=NS_Env_HermesGold_Gold_Mine_OnWalls
 emitters [
  {bEnabled=false, emitterName=HermesGold_Glitter, simTarget=GPUComputeSim, rendererClasses [&NiagaraSpriteRendererProperties]}
  {
   bEnabled=false
   emitterName=HermesGold_Glitter001
   simTarget=GPUComputeSim
   rendererClasses [&NiagaraSpriteRendererProperties]
  }
  {bEnabled=false, emitterName=Parent, simTarget=GPUComputeSim, rendererClasses [&NiagaraSpriteRendererProperties]}
  {bEnabled=false, emitterName=Child, simTarget=GPUComputeSim, rendererClasses [&NiagaraSpriteRendererProperties]}
 ]
 userVariables [
  User.BaseRadiusScale {type=NiagaraFloat, defaultValue=1, description=""}
  User.Landscape {
   type=NiagaraDataInterfaceLandscape
   description=""
   defaultValue {
    struct=NiagaraExt_VariableValue_DataInterface
    value {
     dataInterface=&self:NiagaraDataInterfaceLandscape_0
     dataInterfaceClass=&NiagaraDataInterfaceLandscape
    }
   }
  }
  User.MaxVel {type=NiagaraFloat, defaultValue=100, description=""}
  User.NumberofBeams {type=NiagaraInt32, defaultValue=32}
  User.NumberofRings {type=NiagaraInt32, defaultValue=3}
  User.RadiusSpread {type=NiagaraFloat, defaultValue=1500, description=""}
  User.RingSize {
   type=NiagaraDataInterfaceArrayFloat
   defaultValue {
    struct=NiagaraExt_VariableValue_DataInterface
    value {
     dataInterface=&self:NiagaraDataInterfaceArrayFloat_0
     dataInterfaceClass=&NiagaraDataInterfaceArrayFloat
    }
   }
  }
  User.SplashOffset {
   type=NiagaraDataInterfaceArrayFloat3
   description=""
   defaultValue {
    struct=NiagaraExt_VariableValue_DataInterface
    value {
     dataInterface=&self:NiagaraDataInterfaceArrayFloat3_0
     dataInterfaceClass=&NiagaraDataInterfaceArrayFloat3
    }
   }
  }
  User.WaterBodyIndex {type=NiagaraFloat, defaultValue=0, description=""}
  User.WaterDepth {type=NiagaraFloat, defaultValue=0, description=""}
  User.WaterPlaneLocation {type=Vector3f, description="", defaultValue {struct=Vector3f, value=<0 0 0>}}
  User.WaterPlaneNormal {type=Vector3f, description="", defaultValue {struct=Vector3f, value=<0 0 0>}}
  User.WaveAttenuationFactor {type=NiagaraFloat, defaultValue=0, description=""}
  User.WaveTime {type=NiagaraFloat, defaultValue=0, description=""}
 ]
}
